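# M_DarkGrey is a simple two-sided, Default Lit material in the IceRink package, named for its flat dark grey color and flagged for use with Nanite.
# The graph holds only a constant dark grey color vector and two scalar constants, with no textures or parameters.
# The file does not show which material inputs these constants feed.
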Material /Game/Packages/SF_Fortnite/IceRink/M_DarkGrey
graph {
 properties {bUsedWithNanite=true, pixelDepthOffsetMode=PDOM_Legacy, twoSided=true, usedShadingModels=Default Lit}
 nodes [
  Constant3Vector() {constant=<rgba 0.075 0.075 0.075 0>}
  Constant() {r=5}
  Constant() {r=1}
 ]
}
